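# MetaSound preset for the looping hold sound of loud holdable buttons in the front-end UI.
# It inherits the stereo random looping player from MSS_PlayRandomSoundLooping_Stereo and feeds it one hold-loop wave.
# Short attack and release fades, fixed pitch inputs and slight volume variation shape playback, with source-level pitch raised and volume trimmed.
# It uses the Fort_UI sound class, disables virtualization and overrides concurrency to one instance, stopping the oldest.
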MetaSoundSource /Game/Sounds/Fort_UI_Sounds/FrontEnd/DefaultSounds/HoldableButtons/Holdable_Loud/MSS_UI_Button_Holdable_Loud_Hold_lp
preset_parent=/FortMetasound/Sources/MSS_PlayRandomSoundLooping_Stereo
inputs [
 Attack Time {type=Time}
 bUseRandomStartTime {type=Bool}
 Loop {type=Bool}
 Max Pitch {type=Float}
 Max Volume {type=Float}
 Min Pitch {type=Float}
 Min Volume {type=Float}
 Release Time {type=Time}
 Sounds {type=WaveAsset:Array}
 UE.Source.OnPlay {type=Trigger}
]
interfaces [UE.OutputFormat.Stereo:1.0, UE.Source.OneShot:1.0, UE.Source:1.1]
outputs [
 "UE.OutputFormat.Stereo.Audio:0" {type=Audio}
 "UE.OutputFormat.Stereo.Audio:1" {type=Audio}
 UE.Source.OneShot.OnFinished {type=Trigger}
]
pages [Default {input_defaults {
   Attack Time=0.2
   Loop="<MetaSoundBuilderResult.SUCCEEDED: 0>"
   Max Pitch=1
   Max Volume=1
   Min Pitch=1
   Min Volume=0.9
   Release Time=0.2
   UE.Source.OnPlay=null
   bUseRandomStartTime=null
   Sounds [/Game/Sounds/Fort_UI_Sounds/FrontEnd/DefaultSounds/HoldableButtons/wav/UI_Button_Press_Hold_Loud_Hold_lp]
  }}]
properties {
 assetImportData=&self:AssetImportData
 bOverrideConcurrency=true
 numChannels=2
 outputFormat=Stereo
 pitch=1.35
 sampleRate=48000
 soundClassObject=&Fort_UI
 virtualizationMode=Disabled
 volume=0.755
 concurrencyOverrides {maxCount=1, resolutionRule=StopOldest}
}
